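# Editor utility function library that runs and compares two Atom model processors across a directory of AtomModels.
# CreateAtomMeshesArray creates suffixed duplicates, runs processor 1 on originals and processor 2 on duplicates, then groups output meshes per model, shortest name as main.
# RunAtomModelProcessors overwrites processor slot 0 before running.
# Other functions report triangle counts, budget overruns, bounds and FastBricks comparisons; ProcessAtomModel and its batch wrappers only load, cast and report success.
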EditorUtilityBlueprint /AtomFramework/Editor/Blueprints/AtomProcessorComparisonUtility/Library/EFL_AtomProcessorComparisonUtility
graphs [
 CreateAtomMeshesArray {nodes [
   n1 = CreateAtomMeshesArray!() -> (
    then
    DirectoryPath:String
    Processor_1:Atom Model Processor Class Reference
    Processor_2:Atom Model Processor Class Reference
    Geo1:bool=~
    Geo2:bool=~
    Process Primitives:bool=~
   ) {is=title K2Node_FunctionEntry}
   > n2 = Sequence!(execute) -> (then_0, then_1, then_2:Exec, then_3:Exec, then_4:Exec)
   > Set Processor 1(Processor 1:Atom Model Processor Class Reference=@n1.Processor_1) -> (Output_Get:Atom Model Processor Class Reference)
   Set Processor 2(execute=@n2.then_1, Processor 2:Atom Model Processor Class Reference=@n1.Processor_2) -> (Output_Get:Atom Model Processor Class Reference)
   Set GeometrySettings1(execute=@n2.then_2, GeometrySettings1:bool=@n1.Geo1) -> (Output_Get:bool)
   Set GeometrySettings2(execute=@n2.then_3, GeometrySettings2:bool=@n1.Geo2) -> (Output_Get:bool)
   Set ProcessPrimitives(execute=@n2.then_4, ProcessPrimitives:bool=@n1.Process Primitives) -> (Output_Get:bool)
   > n3 = Set Dir Path(Dir Path:String=@n1) -> (Output_Get:String)
   > CreateFastBricksDuplicates!(execute, DirectoryPath:String=@n3, DuplicateSuffix:String=Get DuplicateSuffix() -> (DuplicateSuffix:String), __WorldContext:Object Reference) -> (then)
   > n4 = ListAssets!(execute, DirectoryPath:String=@n3, bRecursive:bool=true, bIncludeFolder:bool=false) -> (then, ReturnValue:Array of Strings)
   > n5 = Sequence!(execute) -> (then_0, then_1)
   n6 = For Each Loop(Exec=@n5.then_1, Array=@n4)
   n7 = "To String (Integer)"(InInt=Length(TargetArray:Array of Strings=@n4))
   Show Message Dialog(
    execute=@n6.Completed
    Title="To Text (String)"(InString=Processing Complete)
    Message="To Text (String)"(InString=Append!(A="The utility has finished processing ", B=@n7, C:String=" models. Click \"OK\" to view them.") -> (ReturnValue))
    MessageType=Ok
    DefaultValue=No
    MessageCategory=Warning
   )
   > Return Node!(
    execute
    AtomMeshes:Array of STRUCT Atom Processor Comparison Utility Structures=Get TempAtomMeshes() -> (TempAtomMeshes:Array of STRUCT Atom Processor Comparison Utility Structures)
   )
   Set Asset Name(execute=@n6.LoopBody, Asset Name:String=GetBaseFilename(InPath=@n6, bRemovePath=true)) -> (Output_Get:String)
   > n8 = Branch(Condition=OR Boolean!(
     A=Contains(SearchIn=@n6, Substring=SM_, bSearchFromEnd=true)
     B=Contains(SearchIn=@n6, Substring=BP_, bSearchFromEnd=true)
     C:bool=Contains(SearchIn=@n6, Substring=GC_, bSearchFromEnd=true)
    ) -> (ReturnValue))
   n9 = LoadAsset(execute=@n8.else, AssetPath=@n6)
   > n10 = Cast To AtomModel!(execute, Object=@n9) -> (then, CastFailed, AsAtom Model:Atom Model Object Reference) {is=title K2Node_DynamicCast}
   > n11 = Set AtomModel(AtomModel:Atom Model Object Reference=@n10) -> (Output_Get:Atom Model Object Reference)
   n12 = Make Array("[0]":Object Reference=@n11) -> (Array:Array of Object References)
   > UpdateModelAtomPrimitives!(
    execute
    Models:Array of Object References=@n12
    ProcessPrimitives:bool=Get ProcessPrimitives() -> (ProcessPrimitives:bool)
    __WorldContext:Object Reference
   ) -> (then, Sucess:bool)
   > n13 = For Each Loop(Array:Array of Object References=@n12) -> (Array Element:Object Reference)
   > n14 = Cast To AtomModel(Object=@n13)
   > n15 = Set AtomModel(AtomModel=@n14)
   > n16 = Branch(Condition=Contains(SearchIn=GetDisplayName(Object=@n11), Substring=Get DuplicateSuffix(), bSearchFromEnd=true))
   > n17 = RunAtomModelProcessors!(
    execute
    AtomModel:Atom Model Object Reference=@n15
    CustomGeometry:bool=Get GeometrySettings2() -> (GeometrySettings2:bool)
    Processor:Atom Model Processor Class Reference=Get Processor 2() -> (Processor 2:Atom Model Processor Class Reference)
    __WorldContext:Object Reference
   ) -> (then, AtomProcessorResults:Array of Atom Processor Result Structures)
   > n18 = Set AtomProcessorResults(AtomProcessorResults:Array of Atom Processor Result Structures=@n17) -> (Output_Get:Array of Atom Processor Result Structures)
   n19 = For Each Loop(
    Exec=[@n18.then, @n36.then]
    Array:Array of Atom Processor Result Structures=Get AtomProcessorResults() -> (AtomProcessorResults:Array of Atom Processor Result Structures)
   ) -> (Array Element:Atom Processor Result Structure)
   > n20 = For Each Loop(Array:Array of Object References=Break Atom Processor Result!(AtomProcessorResult:"Atom Processor Result Structure (by ref)"=@n19) -> (bSuccess:bool=~, ProcessedObjects:Array of Object References, SharedAssets:Array of Object References) {is=title K2Node_BreakStruct}.ProcessedObjects) -> (Array Element:Object Reference)
   > PrintString(InString=Append!(A="AtomModel processed: ", B=GetObjectName(Object=@n20)) -> (ReturnValue))
   n21 = Get AtomModel() -> (AtomModel:Atom Model Object Reference)
   n22 = Get ModelProcessorInstances!(self:Atom Model Object Reference=@n21) -> (ModelProcessorInstances:Array of Atom Model Processor Instance Structures)
   n23 = For Each Loop(
    Exec=@n19.Completed
    Array:Array of Soft Object References=Break Atom Model Processor Instance!(AtomModelProcessorInstance:"Atom Model Processor Instance Structure (by ref)"="Get (a ref)"(Array:Array of Atom Model Processor Instance Structures=@n22, Dimension 1=0) -> (Output:"Atom Model Processor Instance Structure (by ref)")) -> (bUseCustomSettings:bool=~, ProcessedObjects:Array of Soft Object References) {is=title K2Node_BreakStruct}.ProcessedObjects
   ) -> (Array Element:Soft Object Reference)
   n24 = Get LoadedMeshArray() -> (LoadedMeshArray:Array of Static Mesh Object References)
   n25 = For Each Loop(Exec=@n23.Completed, Array:Array of Static Mesh Object References=@n24) -> (Array Element:Static Mesh Object Reference)
   n26 = For Each Loop(Exec=@n25.Completed, Array=@n24)
   Add(
    execute=@n26.Completed
    TargetArray:Array of STRUCT Atom Processor Comparison Utility Structures=Get TempAtomMeshes()
    NewItem:"STRUCT Atom Processor Comparison Utility Structure (by ref)"=Make STRUCT Atom Processor Comparison Utility!(
     AtomModel_35_#1:Atom Model Soft Object Reference=Make Soft Reference!(Input:Atom Model Object Reference=@n21) -> (Output:Atom Model Soft Object Reference)
     AtomMeshMain_36_#2:Static Mesh Soft Object Reference=Get TempBaseMesh() -> (TempBaseMesh:Static Mesh Soft Object Reference)
     AtomMeshAdditional_38_#3:Array of Static Mesh Soft Object References=Get TempAdditionalMeshes() -> (TempAdditionalMeshes:Array of Static Mesh Soft Object References)
    ) -> (STRUCT_AtomProcessorComparisonUtility:STRUCT Atom Processor Comparison Utility Structure) {is=title K2Node_MakeStruct}
   ) {is=title K2Node_CallArrayFunction}
   > Clear(TargetArray:Array of Static Mesh Soft Object References=Get TempAdditionalMeshes())
   > Set TempBaseMesh(TempBaseMesh:Static Mesh Soft Object Reference) -> (Output_Get:Static Mesh Soft Object Reference)
   > Clear(TargetArray:Array of Static Mesh Object References=Get LoadedMeshArray())
   > Set ShortestMeshNameIndex(ShortestMeshNameIndex:int=0) -> (Output_Get:int=0)
   n27 = Branch(execute=@n26.LoopBody, Condition="Equal (Integer)"(A=@n26.Array Index, B=Get ShortestMeshNameIndex() -> (ShortestMeshNameIndex:int=0)))
   n28 = Make Soft Reference(Input:Static Mesh Object Reference=@n26) -> (Output:Static Mesh Soft Object Reference)
   > Set TempBaseMesh(TempBaseMesh=@n28)
   Add Unique(
    execute=@n27.else
    TargetArray:Array of Static Mesh Soft Object References=Get TempAdditionalMeshes()
    NewItem:"Static Mesh Soft Object Reference (by ref)"=@n28
   )
   n29 = Branch(execute=@n25.LoopBody, Condition="Equal (Integer)"(A=@n25.Array Index))
   n30 = Len(S=GetObjectName(Object=@n25))
   > Set ShortestMeshNameLength(ShortestMeshNameLength:int=@n30) -> (Output_Get:int=0)
   Branch(execute=@n29.else, Condition="integer < integer"(A=@n30, B=Get ShortestMeshNameLength() -> (ShortestMeshNameLength:int=0)))
   > Set ShortestMeshNameLength(ShortestMeshNameLength=@n30) -> (Output_Get=0)
   > Set ShortestMeshNameIndex(ShortestMeshNameIndex=@n25.Array Index) -> (Output_Get=0)
   n31 = LoadAsset_Blocking(execute=@n23.LoopBody, Asset=@n23)
   n32 = GetDisplayName(Object=@n31)
   > PrintString(InString=Append(A="Found object: ", B=@n32))
   n33 = Get AtomModel()
   > Branch(Condition=Contains(
     SearchIn=Split(SourceString=@n32, InStr=SM_, SearchDir=FromStart).RightS
     Substring=GetModelName!(self:Atom Model Object Reference=@n33) -> (ReturnValue:String)
    ))
   > n34 = Cast To StaticMesh(Object=@n31)
   > Add Unique(TargetArray:Array of Static Mesh Object References=Get LoadedMeshArray(), NewItem:"Static Mesh Object Reference (by ref)"=@n34)
   n35 = RunAtomModelProcessors(
    execute=@n16.else
    AtomModel=@n15
    CustomGeometry=Get GeometrySettings1() -> (GeometrySettings1:bool)
    Processor=Get Processor 1() -> (Processor 1:Atom Model Processor Class Reference)
   )
   > n36 = Set AtomProcessorResults(AtomProcessorResults=@n35)
   PrintString(execute=@n14.CastFailed, InString=Could not cast to AtomModel)
   n37 = "To String (Integer)"(InInt=Length(TargetArray=@n4))
   n38 = Get Class Display Name(Class=Get Processor 1())
   n39 = Get Class Display Name(Class=Get Processor_2() -> (Processor_2:Atom Model Processor Class Reference))
   Show Message Dialog(
    execute=@n5.then_0
    Title="To Text (String)"(InString=Atom model processors are about to be run)
    Message="To Text (String)"(InString=Append!(
      A="Click \"OK\" to begin running the Atom model processors.\r\n\r\nThe utility will run the processors "
      B=@n38
      C:String=" and "
      D:String=@n39
      E:String=" on a combined total of "
      F:String=@n37
      G:String=" models.\r\n\r\nThis process will take some time, depending on the number and complexity of models processed. The Unreal Editor will be non-responsive until the processing is complete. Another dialogue will be shown when the processing has completed."
     ) -> (ReturnValue))
    MessageType=Ok
    DefaultValue=No
    MessageCategory=Warning
   )
  ]}
 CreateFastBricksDuplicates {nodes [
   n1 = CreateFastBricksDuplicates!() -> (then, DirectoryPath:String, DuplicateSuffix:String) {is=title K2Node_FunctionEntry}
   > n2 = Set Dir Path(Dir Path=@n1)
   > n3 = ListAssets(DirectoryPath=@n2, bRecursive=true, bIncludeFolder=false)
   > n4 = For Each Loop(Array:Array of Strings=@n3) -> (Array Element:String)
   Return Node!(execute=@n4.Completed)
   n5 = GetBaseFilename(InPath=@n4, bRemovePath=true)
   > PrintString(InString=Append!(A="Checking asset: ", B=@n5) -> (ReturnValue))
   n6 = Contains(SearchIn=@n4, Substring=@n1.DuplicateSuffix, bSearchFromEnd=true)
   n7 = Contains(SearchIn=@n4, Substring=SM_)
   n8 = Contains(SearchIn=@n4, Substring=BP_)
   n9 = Contains(SearchIn=@n4, Substring=GC_)
   > n10 = Branch(Condition=AND Boolean(A=OR Boolean(A=@n7, B=@n8, C=@n9), B=NOT Boolean(A=@n6)))
   n11 = LoadAsset(execute=@n10.else, AssetPath=Append!(A=Get Dir Path() -> (Dir Path:String), B=/, C:String=@n5) -> (ReturnValue))
   > n12 = Branch(Condition=Contains(SearchIn=@n5, Substring=@n1.DuplicateSuffix, bSearchFromEnd=true))
   n13 = DoesAssetExist(execute=@n12.else, AssetPath=Append(A=Get Dir Path(), B=/, C=Append!(A=@n5, B=@n1.DuplicateSuffix) -> (ReturnValue)))
   > n14 = Branch(Condition=@n13)
   n15 = GetAssetTools(execute=@n14.else)
   n16 = Append(A=@n5, B=@n1.DuplicateSuffix)
   > PrintString(InString=Append(A="Creating duplicate: ", B=@n16))
   > Branch(Condition=IsValid(Object=@n11))
   > n17 = DuplicateAsset!(execute, self:Asset Tools Interface=@n15, AssetName:String=@n16, PackagePath:String=Get Dir Path(), OriginalObject:Object Reference=@n11) -> (then, ReturnValue:Object Reference)
   > n18 = Cast To AtomModel(Object=@n17)
   PrintString(execute=@n18.CastFailed, InString=Append(A="Warning: could not cast to AtomModel: ", B=@n5))
  ]}
 GetFilteredMeshes {nodes [
   n1 = GetFilteredMeshes!() -> (then, AtomMeshes:Array of STRUCT Atom Processor Comparison Utility Structures, BudgetedLOD:int, TriangleBudget:int, OnlyOverBudget:bool=~) {is=title K2Node_FunctionEntry}
   > Clear(TargetArray:Array of Strings=Get StaticMeshesNames() -> (StaticMeshesNames:Array of Strings))
   > Clear(TargetArray:Array of Integers=Get StaticMeshesTriCount() -> (StaticMeshesTriCount:Array of Integers))
   > n2 = For Each Loop(Array:Array of STRUCT Atom Processor Comparison Utility Structures=@n1) -> (Array Element:STRUCT Atom Processor Comparison Utility Structure)
   n3 = Branch(execute=@n2.Completed, Condition=@n1.OnlyOverBudget)
   > n4 = Set FilteredDisplayIDs(FilteredDisplayIDs:Array of Integers=Get OverBudgetIDs() -> (OverBudgetIDs:Array of Integers)) -> (Output_Get:Array of Integers)
   Return Node!(
    execute=[@n4.then, @n3.else]
    FilteredMeshIDs:Array of Integers=Get FilteredDisplayIDs() -> (FilteredDisplayIDs:Array of Integers)
    OverBudgetedIDs:Array of Integers=Get OverBudgetIDs()
    HighestTriangleCounts:Array of Integers=Get StaticMeshesTriCount()
    HighestTriangleNames:Array of Strings=Get StaticMeshesNames()
   )
   Set TempTriCount(execute=@n2.LoopBody, TempTriCount:int=0) -> (Output_Get:int=0)
   > n5 = LoadAsset_Blocking(Asset=Break STRUCT Atom Processor Comparison Utility!(STRUCT_AtomProcessorComparisonUtility:"STRUCT Atom Processor Comparison Utility Structure (by ref)"=@n2) -> (
     AtomModel_35_#1:Atom Model Soft Object Reference
     AtomMeshMain_36_#2:Static Mesh Soft Object Reference
     AtomMeshAdditional_38_#3:Array of Static Mesh Soft Object References
    ) {is=title K2Node_BreakStruct}.AtomMeshMain_36_#2)
   > n6 = Cast To StaticMesh(Object=@n5)
   > Set TempTriCount(TempTriCount=int + int(
     A=GetNumTriangles!(self:Static Mesh Object Reference=@n6, LODIndex:int=@n1.BudgetedLOD) -> (ReturnValue:int=0)
     B=Get TempTriCount() -> (TempTriCount:int=0)
    )) -> (Output_Get=0)
   > n7 = Branch(Condition="integer > integer"(A=Get TempTriCount() -> (TempTriCount=0), B=@n1.TriangleBudget))
   > Add Unique(TargetArray:Array of Integers=Get OverBudgetIDs(), NewItem:"Integer (by ref)"=@n2.Array Index)
   > Add(TargetArray:Array of Strings=Get StaticMeshesNames(), NewItem:"String (by ref)"=GetDisplayName(Object=@n6))
   > n8 = Add(TargetArray:Array of Integers=Get StaticMeshesTriCount(), NewItem:"Integer (by ref)"=Get TempTriCount() -> (TempTriCount=0))
   Add Unique(execute=[@n8.then, @n7.else], TargetArray=Get FilteredDisplayIDs(), NewItem=@n2.Array Index)
  ]}
 GetLODTriCounts {nodes [
   n1 = GetLODTriCounts!() -> (then, Processed Objects:Array of Object References) {is=title K2Node_FunctionEntry}
   > n2 = For Each Loop(Array:Array of Object References=@n1) -> (Array Element:Object Reference)
   Return Node!(execute=@n2.Completed, LOD Tri Counts:Array of Integers=Get Temp Tri Counts() -> (Temp Tri Counts:Array of Integers))
   > n3 = Cast To StaticMesh(Object=@n2)
   > n4 = For Loop(LastIndex=int - int(A=GetNumLODs(self=@n3)))
   n5 = Get Temp Tri Counts()
   > Set Array Elem!(
    execute
    TargetArray:Array of Integers=@n5
    Index:int=@n4
    Item:"Integer (by ref)"=int + int(A="Get (a copy)"(Array:Array of Integers=@n5, Dimension 1=@n4) -> (Output:int), B=GetNumTriangles(self=@n3, LODIndex=@n4) -> (ReturnValue=0))
    bSizeToFit:bool=false
   ) -> (then)
  ]}
 GetLargestBound {nodes [
   n1 = GetLargestBound!() -> (then, Array:Array of STRUCT Atom Processor Comparison Utility Structures) {is=title K2Node_FunctionEntry}
   > n2 = For Each Loop(Array:Array of STRUCT Atom Processor Comparison Utility Structures=@n1) -> (Array Element:STRUCT Atom Processor Comparison Utility Structure)
   Return Node!(
    execute=@n2.Completed
    LargestBound:double=Get TempBaseSize() -> (TempBaseSize:double)
    LargestBoundMesh:Static Mesh Object Reference=Get TempStaticMesh() -> (TempStaticMesh:Static Mesh Object Reference)
   )
   > n3 = LoadAsset_Blocking(Asset=Break STRUCT Atom Processor Comparison Utility(STRUCT_AtomProcessorComparisonUtility=@n2).AtomMeshMain_36_#2)
   > n4 = Cast To StaticMesh(Object=@n3)
   n5 = Break Box!(Box:"Box Structure (by ref)"=GetBoundingBox!(self:Static Mesh Object Reference=@n4) -> (ReturnValue:Box Structure)) -> (Min_X:double, Min_Y:double, Min_Z:double, Max_X:double, Max_Y:double, Max_Z:double, IsValid:bool=~) {is=title K2Node_BreakStruct}
   n6 = float - float(A=@n5.Max_Y, B=@n5.Min_Y)
   n7 = float - float(A=@n5.Max_X, B=@n5)
   n8 = Sqrt!(A:double=float + float(A=float * float(A=@n7, B=@n7) {is=title K2Node_PromotableOperator}, B=float * float(A=@n6, B=@n6))) -> (ReturnValue:double)
   > Branch(Condition="float > float"(A=@n8, B=Get TempBaseSize()))
   > Set TempBaseSize(TempBaseSize:double=@n8) -> (Output_Get:double)
   > Set TempStaticMesh(TempStaticMesh:Static Mesh Object Reference=@n4) -> (Output_Get:Static Mesh Object Reference)
  ]}
 GetModelFastBricksTriCounts {nodes [
   GetModelFastBricksTriCounts() {is=title K2Node_FunctionEntry}
   > n1 = GetSelectedAssets()
   > UpdateModelAtomPrimitives(Models=@n1, ProcessPrimitives=false)
   > n2 = For Each Loop(Array:Array of Object References=@n1) -> (Array Element:Object Reference)
   Return Node!(
    execute=@n2.Completed
    Model Tri Counts List:Array of STRUCT Atom Fast Bricks Model Tri Count Structures=Get Model Tri Counts() -> (Model Tri Counts:Array of STRUCT Atom Fast Bricks Model Tri Count Structures)
   )
   > n3 = Cast To AtomModel(Object=@n2)
   > n4 = Set AtomModel(AtomModel=@n3)
   > n5 = RunAtomModelProcessors(AtomModel=@n4, CustomGeometry=false)
   > n6 = For Each Loop(Array:Array of Atom Processor Result Structures=@n5) -> (Array Element:Atom Processor Result Structure)
   > n7 = GetLODTriCounts!(
    execute
    Processed Objects:Array of Object References=Break Atom Processor Result(AtomProcessorResult=@n6) -> (bSuccess=~).ProcessedObjects
    __WorldContext:Object Reference
   ) -> (then, LOD Tri Counts:Array of Integers) {is=title K2Node_CallFunction}
   > Set BaseTriCount(BaseTriCount:Array of Integers=@n7) -> (Output_Get:Array of Integers)
   > n8 = RunAtomModelProcessors(AtomModel=Get AtomModel(), CustomGeometry=true)
   > n9 = For Each Loop(Array=@n8)
   > n10 = GetLODTriCounts(Processed Objects=Break Atom Processor Result(AtomProcessorResult=@n9) -> (bSuccess=~).ProcessedObjects)
   > n11 = Set FastBricksTriCount(FastBricksTriCount:Array of Integers=@n10) -> (Output_Get:Array of Integers)
   > n12 = For Loop(LastIndex=int - int(A=Length(TargetArray:Array of Integers=@n11)))
   n13 = "Get (a copy)"(Array=Get BaseTriCount() -> (BaseTriCount:Array of Integers), Dimension 1=@n12)
   n14 = "Get (a copy)"(Array=Get FastBricksTriCount() -> (FastBricksTriCount:Array of Integers), Dimension 1=@n12)
   n15 = GetModelName(self=Get AtomModel())
   n16 = "To String (Integer)"(InInt=@n12)
   > PrintString(InString=Append!(
     A=Append!(A=@n15, B=" LOD ", C=@n16) -> (ReturnValue)
     B=" Base Triangle Count: "
     C:String="To String (Integer)"(InInt=@n13)
     D:String="  FastBricks Triangle Count: "
     E:String="To String (Integer)"(InInt=@n14)
    ) -> (ReturnValue))
   n17 = Get AtomModel()
   > Add Unique(
    TargetArray:Array of STRUCT Atom Fast Bricks Model Tri Count Structures=Get Model Tri Counts()
    NewItem:"STRUCT Atom Fast Bricks Model Tri Count Structure (by ref)"=Make STRUCT Atom Fast Bricks Model Tri Count!(
     ModelName_2_#4:String=GetModelName(self=@n17)
     BaseTriCount_9_#5:Array of Integers=Get BaseTriCount()
     FastBricksTriCount_10_#6:Array of Integers=Get FastBricksTriCount()
    ) -> (STRUCT_AtomFastBricksModelTriCount:STRUCT Atom Fast Bricks Model Tri Count Structure) {is=title K2Node_MakeStruct}
   )
  ]}
 GetNumValidStaticMeshes {nodes [
   n1 = GetNumValidStaticMeshes!() -> (then, AtomMeshes:Array of STRUCT Atom Processor Comparison Utility Structures) {is=title K2Node_FunctionEntry}
   > n2 = For Each Loop(Array:Array of STRUCT Atom Processor Comparison Utility Structures=@n1) -> (Array Element:STRUCT Atom Processor Comparison Utility Structure)
   Return Node!(execute=@n2.Completed, number of Static Meshes:int=Get Counter() -> (Counter:int=0))
   > n3 = LoadAsset_Blocking(Asset=Break STRUCT Atom Processor Comparison Utility(STRUCT_AtomProcessorComparisonUtility=@n2).AtomMeshMain_36_#2)
   > Cast To StaticMesh(Object=@n3)
   > Increment Int(Value=Get Counter() -> (Counter=0))
  ]}
 GetRadiusInfo {nodes [
   n1 = GetRadiusInfo!() -> (then, AtomMeshes:Array of STRUCT Atom Processor Comparison Utility Structures) {is=title K2Node_FunctionEntry}
   > n2 = For Each Loop(Array=@n1)
   > n3 = LoadAsset_Blocking(Asset=Break STRUCT Atom Processor Comparison Utility(STRUCT_AtomProcessorComparisonUtility=@n2).AtomMeshMain_36_#2)
   > n4 = Cast To StaticMesh(Object=@n3)
   > n5 = Branch(Condition="Equal (Integer)"(A=@n2.Array Index))
   n6 = BreakBoxSphereBounds!(InBoxSphereBounds:"Box Sphere Bounds Structure (by ref)"=GetBounds!(self:Static Mesh Object Reference=@n4) -> (ReturnValue:Box Sphere Bounds Structure)) -> (Origin:Vector, BoxExtent:Vector, SphereRadius:float)
   > n7 = Set TempSmallest(TempSmallest:double=@n6.SphereRadius) -> (Output_Get:double)
   n8 = Branch(execute=[@n7.then, @n5.else], Condition="float < float"(A=@n6.SphereRadius, B=Get TempSmallest() -> (TempSmallest:double)))
   > n9 = Set TempSmallest(TempSmallest=@n6.SphereRadius)
   n10 = Branch(execute=[@n9.then, @n8.else], Condition="float > float"(A=@n6.SphereRadius, B=Get TempLargest() -> (TempLargest:double)))
   > n11 = Set TempLargest(TempLargest:double=@n6.SphereRadius) -> (Output_Get:double)
   Set SumRadiuses(execute=[@n11.then, @n10.else], SumRadiuses:double=float + float(A=@n6.SphereRadius, B=Get SumRadiuses() -> (SumRadiuses:double))) -> (Output_Get:double)
   n12 = Length(TargetArray:Array of STRUCT Atom Processor Comparison Utility Structures=@n1)
   n13 = Get SumRadiuses()
   Return Node!(
    execute=@n2.Completed
    AverageRadius:int=Round(A=float / float(A=@n13, B:int=@n12)) {is=title K2Node_CallFunction}
    LargestRadius:int=Round(A=Get TempLargest())
    SmallestRadius:int=Round(A=Get TempSmallest())
   )
  ]}
 GetTriangleCountAtomMesh {nodes [
   n1 = GetTriangleCountAtomMesh!() -> (then, AtomMeshes:Array of STRUCT Atom Processor Comparison Utility Structures, LODIndex:int) {is=title K2Node_FunctionEntry}
   > n2 = For Each Loop(Array=@n1)
   n3 = Get MeshCounter() -> (MeshCounter:int=0)
   n4 = Get TotalTriangles() -> (TotalTriangles:int=0)
   Return Node!(execute=@n2.Completed, TotalTriangleCount:int=@n4, AverageTriangleCount:int=int / int(A=@n4, B=@n3), NumModels:int=@n3)
   > n5 = LoadAsset_Blocking(Asset=Break STRUCT Atom Processor Comparison Utility(STRUCT_AtomProcessorComparisonUtility=@n2).AtomMeshMain_36_#2)
   > n6 = Cast To StaticMesh(Object=@n5)
   > Increment Int(Value=Get MeshCounter() -> (MeshCounter=0))
   > Set TotalTriangles(TotalTriangles:int=int + int(A=Get TotalTriangles() -> (TotalTriangles=0), B=GetNumTriangles(self=@n6, LODIndex=@n1.LODIndex) -> (ReturnValue=0))) -> (Output_Get:int=0)
  ]}
 ProcessAtomModel {nodes [
   n1 = ProcessAtomModel!() -> (then, Atom Model:Atom Model Soft Object Reference, ShowDialog:bool=~) {is=title K2Node_FunctionEntry}
   > n2 = LoadAsset_Blocking(Asset=@n1)
   > n3 = Cast To AtomModel(Object=@n2)
   > n4 = Branch(Condition=@n1.ShowDialog)
   n5 = Append(A="Done! \r\n\r\n", B=GetDisplayName(Object=@n3), C=" has been Processed.\r\n\r\n")
   > n6 = Show Message Dialog(
    Title=MakeLiteralText(Value="LOCTEXT(\"Process Selected Atom Model\")")
    Message=MakeLiteralText(Value="To Text (String)"(InString=@n5))
    MessageType=Ok
    DefaultValue=Ok
    MessageCategory=Success
   )
   n7 = Set Success(execute=[@n6.then, @n4.else], Success:bool=true) -> (Output_Get:bool)
   Return Node!(execute=[@n7.then, @n8.then], Success?:bool=Get Success() -> (Success:bool))
   n8 = Set Success(execute=@n3.CastFailed, Success=false)
  ]}
 ProcessAtomModels {nodes [
   n1 = ProcessAtomModels!() -> (then, ShowDialog:bool=~, AtomMeshes:Array of STRUCT Atom Processor Comparison Utility Structures) {is=title K2Node_FunctionEntry}
   > n2 = For Each Loop(Array=@n1.AtomMeshes)
   n3 = Branch(execute=@n2.Completed, Condition=@n1)
   n4 = Append(A="Done! \r\n\r\n", B="To String (Integer)"(InInt=Get NumModels() -> (NumModels:int=0)), C=" Atom Models Processed.")
   > n5 = Show Message Dialog(
    Title=MakeLiteralText(Value="LOCTEXT(\"Batch Process Atom Assets\")")
    Message=MakeLiteralText(Value="To Text (String)"(InString=@n4))
    MessageType=Ok
    DefaultValue=Ok
    MessageCategory=Success
   )
   Return Node!(execute=[@n5.then, @n3.else], NumAssets:int=Get NumModels() -> (NumModels=0))
   n6 = LoadAsset_Blocking(execute=@n2.LoopBody, Asset=Break STRUCT Atom Processor Comparison Utility(STRUCT_AtomProcessorComparisonUtility=@n2))
   > n7 = Cast To AtomModel(Object=@n6)
   > n8 = ProcessAtomModel!(
    execute
    Atom Model:Atom Model Soft Object Reference=Make Soft Reference(Input:Atom Model Object Reference=@n7) -> (Output:Atom Model Soft Object Reference)
    ShowDialog:bool=false
    __WorldContext:Object Reference
   ) -> (then, Success?:bool) {is=title K2Node_CallFunction}
   > Branch(Condition=@n8)
   > Increment Int(Value=Get NumModels() -> (NumModels=0))
  ]}
 ProcessSelectedAtomModels {nodes [
   ProcessSelectedAtomModels() {is=title K2Node_FunctionEntry}
   > n1 = GetSelectedAssetsOfClass(AssetClass=/Script/AtomRuntime.AtomModel)
   > n2 = For Each Loop(Array:Array of Object References=@n1) -> (Array Element:Object Reference)
   n3 = Append(A="Done! \r\n", B="To String (Integer)"(InInt=Get NumModels() -> (NumModels=0)), C=" Atom Models Processed!")
   Show Message Dialog(
    execute=@n2.Completed
    Title=MakeLiteralText(Value="LOCTEXT(\"Process Selected Atom Model\")")
    Message=MakeLiteralText(Value="To Text (String)"(InString=@n3))
    MessageType=Ok
    DefaultValue=Ok
    MessageCategory=Success
   )
   > Return Node!(execute)
   n4 = Cast To AtomModel(execute=@n2.LoopBody, Object=@n2)
   > n5 = ProcessAtomModel(Atom Model=Make Soft Reference(Input=@n4), ShowDialog=true)
   > Branch(Condition=@n5)
   > Increment Int(Value=Get NumModels() -> (NumModels=0))
  ]}
 RunAtomModelProcessors {nodes [
   n1 = RunAtomModelProcessors!() -> (then, AtomModel:Atom Model Object Reference, CustomGeometry:bool=~, Processor:Atom Model Processor Class Reference) {is=title K2Node_FunctionEntry}
   > PrintString(InString=Append!(A="Running processor ", B=Get Class Display Name(Class=@n1.Processor), C=" on model ", D:String=GetDisplayName(Object=@n1)) -> (ReturnValue))
   > n2 = Construct Object from Class(Class=@n1.Processor, self=@n1) -> (ReturnValue:Atom Model Processor Object Reference)
   n3 = Make Atom Model Processor Instance!(bUseCustomSettings:bool=@n1.CustomGeometry) -> (AtomModelProcessorInstance:Atom Model Processor Instance Structure) {is=title K2Node_MakeStruct}
   > SetModelProcessor!(
    execute
    ProcessorInstance:"Atom Model Processor Instance Structure (by ref)"=@n3
    ModelProcessor:Atom Model Processor Object Reference=@n2
    bUseCustomSettings:bool=@n1.CustomGeometry
   ) -> (then)
   > Set Array Elem(
    TargetArray:Array of Atom Model Processor Instance Structures=Get ModelProcessorInstances(self=@n1)
    Index=0
    Item:"Atom Model Processor Instance Structure (by ref)"=@n3
    bSizeToFit=true
   )
   > n4 = RunModelProcessors!(execute, AtomModel:Atom Model Object Reference=@n1) -> (then, ReturnValue:Array of Atom Processor Result Structures)
   > Return Node!(execute, AtomProcessorResults:Array of Atom Processor Result Structures=@n4)
  ]}
 UpdateModelAtomPrimitives {nodes [
   n1 = UpdateModelAtomPrimitives!() -> (then, Models:Array of Object References, ProcessPrimitives:bool=~) {is=title K2Node_FunctionEntry}
   > n2 = Branch(Condition=@n1.ProcessPrimitives)
   > n3 = For Each Loop(Array=@n1)
   > n4 = Cast To AtomModel(Object=@n3)
   > n5 = For Each Loop!(
    Exec
    Array:Array of Atom Model Part Instance Info Structures=GetParts!(PartsCollection:"Atom Model Parts Collection Structure (by ref)"=GetPartsCollection!(self:Atom Model Object Reference=@n4) -> (ReturnValue:Atom Model Parts Collection Structure)) -> (ReturnValue:Array of Atom Model Part Instance Info Structures)
   ) -> (
    LoopBody
    Array Element_PartGuid:Atom Model Part Guid Structure
    Array Element_PartId:int
    Array Element_Primitive:Atom Primitive Object Reference
    Array Element_Transforms:Array of Transforms
    Array Element_Decorations:Array of Atom Model Part Decoration Info Structures
    Array Element_Colors:Array of Atom Model Part Color Info Structures
    Array Element_SelectionSets:Array of Strings
    Array Element_Group:String
    Array Element_ParentGroup:String
    Array Element_CommonParts:Array of Atom Common Part And Transform Structures
    Array Element_bIsUndersideVisible:bool=~
    Array Index
    Completed
   )
   n6 = Get Atom Primitives() -> (Atom Primitives:Array of Atom Primitive Object References)
   n7 = UpdatePrimitiveArray!(execute=@n5.Completed, AtomPrimitives:Array of Atom Primitive Object References=@n6) -> (then, ReturnValue:bool)
   > Return Node!(execute, Sucess:bool=@n7)
   Add Unique(
    execute=@n5.LoopBody
    TargetArray:Array of Atom Primitive Object References=@n6
    NewItem:"Atom Primitive Object Reference (by ref)"=@n5.Array Element_Primitive
   )
   Return Node(execute=@n2.else, Sucess=true)
  ]}
]
variables []
